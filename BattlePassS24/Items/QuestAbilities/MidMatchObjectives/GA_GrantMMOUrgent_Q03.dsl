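; Gameplay ability that grants an urgent Season 24 mid-match objective quest (Q03) to the owning player.
; On activation it waits for a tagged gameplay event on the player controller, grants the transient quest and binds a failure delegate that ends the ability for that quest.
; A completion handler, unbound on end but never bound here, matches the quest by asset ID and sends a failsafe capture-point completion stat.
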(event Ability|EventFailedToActivatePassiveAbility
  (params (ActorInfo :type Gameplay%20Ability%20Actor%20Info%20Structure))
  (|CallParentFunction))

(event Ability|EventActivateAbility
  (|CallParentFunction)
  (bind _asfort_player_controller_athena (Utilities|Casting|CastToFortPlayerControllerAthena (Fortnite|GetFortPlayerControllerfromActor (Variables|GAVars|Data|GetOwningPlayerPawn))))
  (Ability|Async|WaitGameplayEventToActor _asfort_player_controller_athena (Variables|Default|GetTag)
    (:EventReceived
      (Class|FortControllerComponentTransientQuests|GrantTransientQuest (Class|FortPlayerControllerAthena|GetTransientQuestsComponent _asfort_player_controller_athena) (Variables|Default|GetQuest%20Def%20to%20Grant) 0)
      (Development|PrintString "Urgentquest granted" true true "(R=1.000000,G=0.000000,B=0.192244,A=1.000000)")
      ;[0 0 0] If quest fails, end the ability
      (Quests|BindEventto%28DEPRECATED%29OnUrgentTransientQuestFailed (Class|FortQuestManager|GetQuestManagerComponent (Variables|GAVars|Data|GetOwningPlayerQuestManager) "/Script/FortniteGame.FortQuestManagerComponent_Athena") (delegate Custom|OnTransientQuestFailed)))))

(event Custom|OnTransientQuestFailed
  (params (TransientQuestItem :type Fort%20Quest%20Item%20Object%20Reference))
  ;[0 0 0] If quest fails, end the ability
  (if (== (AssetManager|GetPrimaryAssetIdfromObject (Quest|GetQuestDefinition TransientQuestItem)) (AssetManager|GetPrimaryAssetIdfromObject (Variables|Default|GetQuest%20Def%20to%20Grant)))
    (Ability|EndAbility)))

(event Custom|OnUrgentQuestCompleted
  (params (TransientQuestItem :type Fort%20Quest%20Item%20Object%20Reference))
  ;[0 0 0] When any urgent quest is completed, confirm that it is this urgent quest. Then send a failsafe stat event to the urgent quest to ensure it completes
  (if (== (AssetManager|GetPrimaryAssetIdfromObject (Quest|GetQuestDefinition TransientQuestItem)) (AssetManager|GetPrimaryAssetIdfromObject (Variables|Default|GetQuest%20Def%20to%20Grant)))
    (Quest|SendComplexCustomStatEvent (Quests|GetQuestManager (Fortnite|GetFortPlayerControllerfromActor (Variables|GAVars|Data|GetOwningPlayerPawn)) "Invalid") 0 "(GameplayTags=((TagName=\"Athena.Quests.S24.MMO.CapturePoint.Complete\")))" "(GameplayTags=)" 1)
    (Ability|EndAbility)))

(event Ability|EventOnEndAbility
  (params (bWasCancelled :type Boolean))
  (bind _owningplayerquestmanager (Variables|GAVars|Data|GetOwningPlayerQuestManager))
  (bind _returnvalue (Class|FortQuestManager|GetQuestManagerComponent _owningplayerquestmanager "/Script/FortniteGame.FortQuestManagerComponent_Athena"))
  ;[0 0 0] Clean up event binds when the ability ends
  (|CallParentFunction bWasCancelled)
  (Quests|UnbindEventfromOnUrgentTransientQuestCompleted _returnvalue (EventDispatchers|CreateEvent :!event Custom|OnUrgentQuestCompleted))
  (Quests|UnbindEventfrom%28DEPRECATED%29OnUrgentTransientQuestFailed _returnvalue (EventDispatchers|CreateEvent :!event Custom|OnTransientQuestFailed)))
